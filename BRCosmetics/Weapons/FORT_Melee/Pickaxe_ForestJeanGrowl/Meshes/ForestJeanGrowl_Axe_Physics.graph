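# Physics asset for the ForestJeanGrowl pickaxe that defines simulation bodies for three dangling lace chains (center, left and right, the last named "race").
# Each chain has a small base capsule followed by three larger dyn capsules, with each dyn body constrained to the one before it.
# Swing limits tighten from root to tip while twist is free, so each chain bends most at its attachment point.
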PhysicsAsset /BRCosmetics/Weapons/FORT_Melee/Pickaxe_ForestJeanGrowl/Meshes/ForestJeanGrowl_Axe_Physics
bodies {
 base_lace [{shapeType=Capsule, length=0.505, radius=0.505, shapeName=base_lace_capsule, rotation=<rot 0 -90 0>}]
 base_lace_l [{shapeType=Capsule, length=0.505, radius=0.505, shapeName=base_lace_l_capsule, rotation=<rot 0 -90 0>}]
 base_race_r [{shapeType=Capsule, length=0.505, radius=0.505, shapeName=base_race_r_capsule, rotation=<rot 0 -90 0>}]
 dyn_lace [{shapeType=Capsule, length=2, radius=1, shapeName=dyn_lace_capsule, center=<2 0 0>, rotation=<rot 90 0 0>}]
 dyn_lace_00 [{shapeType=Capsule, length=2, radius=1, shapeName=dyn_lace_00_capsule, center=<2 0 0>, rotation=<rot 90 0 0>}]
 dyn_lace_00_l [{shapeType=Capsule, length=2, radius=1, shapeName=dyn_lace_00_l_capsule, center=<2 0 0>, rotation=<rot 90 0 0>}]
 dyn_lace_01 [{shapeType=Capsule, length=2, radius=1, shapeName=dyn_lace_01_capsule, center=<2 0 0>, rotation=<rot 90 0 0>}]
 dyn_lace_01_l [{shapeType=Capsule, length=2, radius=1, shapeName=dyn_lace_01_l_capsule, center=<2 0 0>, rotation=<rot 90 0 0>}]
 dyn_lace_l [{shapeType=Capsule, length=2, radius=1, shapeName=dyn_lace_l_capsule, center=<2 0 0>, rotation=<rot 90 0 0>}]
 dyn_race_00_r [{shapeType=Capsule, length=2, radius=1, shapeName=dyn_race_00_r_capsule, center=<2 0 0>, rotation=<rot 90 0 0>}]
 dyn_race_01_r [{shapeType=Capsule, length=2, radius=1, shapeName=dyn_race_01_r_capsule, center=<2 0 0>, rotation=<rot 90 0 0>}]
 dyn_race_r [{shapeType=Capsule, length=2, radius=1, shapeName=dyn_race_r_capsule, center=<2 0 0>, rotation=<rot 90 0 0>}]
}
constraints [
 {bone1Name=dyn_lace, bone2Name=base_lace, swing1LimitDegrees=120, swing2LimitDegrees=120, twistLimitDegrees=45, twistMotion=Free}
 {bone1Name=dyn_lace_00, bone2Name=dyn_lace, swing1LimitDegrees=90, swing2LimitDegrees=90, twistLimitDegrees=45, twistMotion=Free}
 {bone1Name=dyn_lace_01, bone2Name=dyn_lace_00, swing1LimitDegrees=60, swing2LimitDegrees=60, twistLimitDegrees=45, twistMotion=Free}
 {bone1Name=dyn_lace_l, bone2Name=base_lace_l, swing1LimitDegrees=120, swing2LimitDegrees=120, twistLimitDegrees=45, twistMotion=Free}
 {bone1Name=dyn_lace_00_l, bone2Name=dyn_lace_l, swing1LimitDegrees=90, swing2LimitDegrees=90, twistLimitDegrees=45, twistMotion=Free}
 {bone1Name=dyn_lace_01_l, bone2Name=dyn_lace_00_l, swing1LimitDegrees=60, swing2LimitDegrees=60, twistLimitDegrees=45, twistMotion=Free}
 {bone1Name=dyn_race_r, bone2Name=base_race_r, swing1LimitDegrees=120, swing2LimitDegrees=120, twistLimitDegrees=45, twistMotion=Free}
 {bone1Name=dyn_race_00_r, bone2Name=dyn_race_r, swing1LimitDegrees=90, swing2LimitDegrees=90, twistLimitDegrees=45, twistMotion=Free}
 {bone1Name=dyn_race_01_r, bone2Name=dyn_race_00_r, swing1LimitDegrees=60, swing2LimitDegrees=60, twistLimitDegrees=45, twistMotion=Free}
]
